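; Skyfire gameplay ability that gives "can't do that" feedback when the player presses primary fire while firing is prevented.
; On activation it caches the avatar's Fort player pawn, then waits for the Event.Primary.Pressed gameplay event.
; It plays the CantDoThat montage if blocking tags match without prison-montage or air-control-block tags.
; Otherwise it plays a 2D can't-fire sound where cosmetic events can execute.
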(event Ability|EventActivateAbility
  ;[0 1 0.19] Wait for Primary input to be pressed.
  (|CallParentFunction)
  (bind _outfortplayerpawn (CallFunction|GetAvatarasValidFortPlayerPawn)
    (:ValidFortPlayerPawn
      (Variables|Default|SetPlayerPawn _outfortplayerpawn)
      (Ability|Tasks|WaitGameplayEvent "(TagName=\"Event.Primary.Pressed\")"
        (:EventReceived
          (CallFunction|PlayPreventFireFX))))))

(event Custom|PlayPreventFireFX
  (bind (_case_0_true _case_1_true _case_2_true) (GameplayTags|TagContainer|CompareTagAssetInterfacetoOtherTagContainers (Variables|Default|GetTC_BlockingTags) (Variables|Default|GetTC_PrisonMontagePlaying) (Variables|Default|GetTC_BlockAirControl) (Ability|GetAbilitySystemComponent (Variables|Default|GetPlayerPawn))))
  (bind _playerpawn (Variables|Default|GetPlayerPawn))
  ;[1 0 0.9929] Play "you can't do that" montage and FX.
  (Utilities|IsValid _playerpawn
    (:Is%20Valid
      (if (Math|Boolean|ANDBoolean _case_0_true (not _case_1_true) (not _case_2_true))
        (Ability|Tasks|PlayMontageandWaitforTrigger "None" "/Game/Animation/Game/MainPlayer/Montages/General/CantDoThat.CantDoThat")
        (Networking|CanExecuteCosmeticEvents
          (:True
            (Audio|SpawnSound2D (Variables|Default|GetCantFireSFX) (Variables|Default|GetCantFireVolume))))))))
