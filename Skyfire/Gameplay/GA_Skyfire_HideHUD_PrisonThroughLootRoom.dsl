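; Skyfire gameplay ability that hides selected HUD elements while active, named for the stretch from prison through loot room.
; On activation it calls the parent, then hides the elements listed in the TC_HudElementsToHide tag container through the Fort HUD context.
; When the ability ends, it re-enables the same elements, whether or not it was cancelled.
; Both paths act only if the HUD context is valid.
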(event Ability|EventActivateAbility
  (bind _returnvalue (BlueprintContext|GetFortHUDContext))
  ;[0 1 0.2292] Disable HUD elements by tag.
  (|CallParentFunction)
  (Utilities|IsValid _returnvalue
    (:Is%20Valid
      (HUD|SetHUDElementVisibility _returnvalue (Variables|Default|GetTC_HudElementsToHide) true "None"))))

(event Ability|EventOnEndAbility
  (params (bWasCancelled :type Boolean))
  (bind _returnvalue (BlueprintContext|GetFortHUDContext))
  ;[1 0.237 0.9724] Re-enable the elements when ability is removed.
  (Utilities|IsValid _returnvalue
    (:Is%20Valid
      (HUD|SetHUDElementVisibility _returnvalue (Variables|Default|GetTC_HudElementsToHide) false "None"))))
