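; Blueprint for a prison squad point actor in Skyfire's gameplay.
; On BeginPlay it calls the parent logic, then enables actor collision so its TempInvisCollision works.
; Tick and the construction script are empty, so any other behavior comes from the parent class.
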(event EventBeginPlay
  (|CallParentFunction)
  (Collision|SetActorEnableCollision true) ; Turn on collision so the TempInvisCollision works
)

(event EventTick
  (params (DeltaSeconds :type Float%20%28single-precision%29)))

(fn ConstructionScript)
