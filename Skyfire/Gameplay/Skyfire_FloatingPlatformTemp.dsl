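; Skyfire_FloatingPlatformTemp is a Blueprint, named as a temporary floating platform, whose file holds only a construction script.
; The script assigns the Static Mesh variable to the StaticMeshComponent, so the displayed mesh follows that variable.
; No movement, floating behavior or other gameplay logic appears here.
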(fn ConstructionScript
  (Components|StaticMesh|SetStaticMesh (Variables|Default|GetStaticMeshComponent) (Variables|Default|GetStatic%20Mesh)))
