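; Teleport squad point actor that launches a player forward after teleporting on the Mothership.
; On teleport completion it validates the player state and its current pawn, then stores that pawn.
; The pawn's forward X and Y are scaled by a launch multiplier and combined with a separate launch height.
; LaunchCharacter applies this vector, overriding existing horizontal and vertical velocity.
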(event EventBeginPlay)

(event EventTick
  (params (DeltaSeconds :type Float%20%28single-precision%29)))

(event EventOnTeleportComplete
  (params (Player :type Fort%20Player%20State%20Athena%20Object%20Reference))
  (bind (_returnvalue_x _returnvalue_y) (Math|Vector|BreakVector (Transformation|GetActorForwardVector (Variables|Default|GetPlayerPawn))))
  (bind _returnvalue (PlayerState|GetCurrentPawn Player))
  ;[0.1892 0.3638 1] Launch player forward after teleporting on the Mothership
  (Utilities|IsValid Player
    (:Is%20Valid
      (Utilities|IsValid _returnvalue
        (:Is%20Valid
          (bind _output_get (Variables|Default|SetPlayerPawn _returnvalue))
          (Character|LaunchCharacter _output_get (Math|Vector|MakeVector (* _returnvalue_x (Variables|Default|GetLaunchMultiplier)) (* (Variables|Default|GetLaunchMultiplier) _returnvalue_y) (Variables|Default|GetLaunchHeight)) true true))))))

(fn ConstructionScript)
